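# Physics asset for the Vampire backpack cosmetic, giving its skeletal mesh simple capsule bodies for the root and both wings.
# Each wing is a three-body chain from elbow through two pinky bones, with the elbow constrained to the backpack root.
# Every joint locks swing 1 and twist, leaving swing 2 as the only unlocked angular axis, so each joint bends like a hinge.
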PhysicsAsset /BRCosmetics/Accessories/FORT_Backpacks/F_MED_Vampire_Backpack/Meshes/F_MED_Vampire_Backpack_Physics
bodies {
 backpack_root [{shapeType=Capsule, length=14.6214, radius=6.8828, center=<0.7892 -3.6115 0>, rotation=<rot 0 -90 0>}]
 wing_elbow_l [{shapeType=Capsule, length=17.4667, radius=3.043, center=<7.3522 -0.2961 -0.1174>, rotation=<rot 0 92.7153 -97.2604>}]
 wing_elbow_r [{shapeType=Capsule, length=17.4667, radius=3.043, center=<-7.3522 0.2961 0.1174>, rotation=<rot 0 92.7154 -97.2604>}]
 wing_pinky_01_l [{shapeType=Capsule, length=20.2685, radius=3.1951, center=<11.4315 0.3275 0.7363>, rotation=<rot 0 -90.5203 -86.876>}]
 wing_pinky_01_r [{shapeType=Capsule, length=20.2685, radius=3.1951, center=<-11.4315 0.3275 -0.7363>, rotation=<rot 0 -90.5203 -86.876>}]
 wing_pinky_02_l [{shapeType=Capsule, length=23, radius=5, center=<12.86 0.21 0.08>, rotation=<rot 0 -87 -83>}]
 wing_pinky_02_r [{shapeType=Capsule, length=23, radius=5, center=<-12.86 0.21 0.08>, rotation=<rot 0 -87 -83>}]
}
constraints [
 {
  bone1Name=wing_elbow_l
  bone2Name=backpack_root
  swing1LimitDegrees=45
  swing1Motion=Locked
  swing2LimitDegrees=45
  twistLimitDegrees=45
  twistMotion=Locked
 }
 {
  bone1Name=wing_elbow_r
  bone2Name=backpack_root
  swing1LimitDegrees=45
  swing1Motion=Locked
  swing2LimitDegrees=45
  twistLimitDegrees=45
  twistMotion=Locked
 }
 {
  bone1Name=wing_pinky_01_r
  bone2Name=wing_elbow_r
  swing1LimitDegrees=45
  swing1Motion=Locked
  swing2LimitDegrees=45
  twistLimitDegrees=45
  twistMotion=Locked
 }
 {
  bone1Name=wing_pinky_01_l
  bone2Name=wing_elbow_l
  swing1LimitDegrees=45
  swing1Motion=Locked
  swing2LimitDegrees=45
  twistLimitDegrees=45
  twistMotion=Locked
 }
 {
  bone1Name=wing_pinky_02_l
  bone2Name=wing_pinky_01_l
  swing1LimitDegrees=45
  swing1Motion=Locked
  swing2LimitDegrees=45
  twistLimitDegrees=45
  twistMotion=Locked
 }
 {
  bone1Name=wing_pinky_02_r
  bone2Name=wing_pinky_01_r
  swing1LimitDegrees=45
  swing1Motion=Locked
  swing2LimitDegrees=45
  twistLimitDegrees=45
  twistMotion=Locked
 }
]
